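; Thrown orb projectile for the SpaceClique weapon that homes, bounces, damages pawns, vehicles and buildings, then returns to its owner like a boomerang.
; Hit and overlap events feed server-side damage through gameplay effects.
; Timers, bounce limits, pawn hits or manual recall (with radial damage) trigger the return, tuned by scalable floats.
; Short hits pop up and stall before returning; KillProjectile sends the catch event and destroys the orb.
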(event Collision|EventHit
  (params (MyComp :type Primitive%20Component%20Object%20Reference) (Other :type Actor%20Object%20Reference) (OtherComp :type Primitive%20Component%20Object%20Reference) (bSelfMoved :type Boolean) (HitLocation :type Vector :default "0, 0, 0") (HitNormal :type Vector :default "0, 0, 0") (NormalImpulse :type Vector :default "0, 0, 0") (Hit :type Hit%20Result%20Structure%20%28by%20ref%29))
  (bind (_bblockinghit _binitialoverlap _time _distance _location _impactpoint _normal _impactnormal _physmat _hitactor) (Collision|BreakHitResult Hit))
  ;[1 1 1] Event Hit: Trigger a ShieldSphereTrace.
  (|CallParentFunction MyComp Other OtherComp bSelfMoved HitLocation HitNormal NormalImpulse Hit) ; TODO// consider increasing travel duration with each bounce
  (Variables|HitInfo|SetBoomerangHitNormal HitNormal)
  (Variables|HitInfo|SetBoomerangHitLoc HitLocation)
  (Projectile|SetHomingTarget 0 self 0 0 :!class /Script/FortniteGame.FortProjectileMovementInterface)
  (bind _output_get (Variables|HitInfo|SetHitActor _hitactor))
  (bind _asplayer_pawn_athena (Utilities|Casting|CastToPlayerPawn_Athena _output_get)
    (:then
      (Utilities|Array|AddUnique (Variables|Targeting|GetActorsToIgnoreForTargeting) _asplayer_pawn_athena)
      (Variables|Targeting|SetbTargetAcquired true)
      (CallFunction|PlayHitPawnSounds _asplayer_pawn_athena))
    (:CastFailed
      (Audio|PlaySoundatLocation (Variables|Audio|GetHitEnvSoundCue) (Variables|HitInfo|GetBoomerangHitLoc))))
  (Utilities|FlowControl|Sequence
    (:then_0
      (if (not (Variables|Default|GetReturnToSender))
        (Variables|Targeting|SetHomingTarget 0))
      (CallFunction|HitCheckEvent Hit))
    (:then_1
      (if (not (Variables|Default|GetReturnToSender))
        (Variables|Default|SetCurrentNumberOfBounces (+ (Variables|Default|GetCurrentNumberOfBounces) 1))
        (CallFunction|BounceandMirrorVectorNormal)))))

(event EventOnStop
  (params (Hit :type Hit%20Result%20Structure%20%28by%20ref%29)))

(event EventOnBounce
  (params (Hit :type Hit%20Result%20Structure%20%28by%20ref%29)))

(event OnComponentBeginOverlap(CapsuleComponent)
  (params (OverlappedComponent :type Primitive%20Component%20Object%20Reference) (OtherActor :type Actor%20Object%20Reference) (OtherComp :type Primitive%20Component%20Object%20Reference) (OtherBodyIndex :type Integer) (bFromSweep :type Boolean) (SweepResult :type Hit%20Result%20Structure%20%28by%20ref%29)))

(event OnProjectileStop(ProjectileMovementComponent)
  (params (ImpactResult :type Hit%20Result%20Structure%20%28by%20ref%29))
  ;[0.6406 1 0.6843] EVENT -> On Projectile Stop
  (CallFunction|HitCheckEvent ImpactResult))

(event Custom|Hit%20Check%20Event
  (params (Hit%20Result :type Hit%20Result%20Structure))
  (bind _returnvalue (Utilities|GetDisplayName (.HitActor (Collision|BreakHitResult Hit%20Result))))
  (bind (_bblockinghit _binitialoverlap _time _distance _location _impactpoint _normal _impactnormal _physmat _hitactor) (Collision|BreakHitResult Hit%20Result))
  (bind _lasthitobstacle (Variables|BPSetVariables|GetLastHitObstacle))
  ;[0.6406 1 0.6843] EVENT -> On Projectile Stop
  (Niagara|SetNiagaraVariable(Bool) (Variables|Default|GetNiagara) "ImpactObject" true)
  (Networking|SwitchHasAuthority
    (:Authority
      (Utilities|IsValid _hitactor
        (:Is%20Valid
          (if (Variables|Actor|GetbCanBeDamaged _hitactor)
            (Variables|HitInfo|SetHitResult Hit%20Result)
            (bind _asfort_vehicle_interface (Utilities|Casting|CastToFortVehicleInterface _hitactor)
              (:then
                ;;[0.1752 1 0.4401] Hit and Damage Vehicles
                (CallFunction|DamageHitActor (Vehicle|GetActor _asfort_vehicle_interface))
                (Collision|IgnoreActorwhenMoving (Variables|Collision|GetCapsuleComponent) (Vehicle|GetActor _asfort_vehicle_interface) true)
                (if (Variables|Default|GetReturnToSender)
                  (Default|HandleProjectileStopReturn)))
              (:CastFailed
                ;;[1 0.4908 0.4609] Hit and Damaged Building
                (bind _asbuilding_actor (Utilities|Casting|CastToBuildingActor _hitactor)
                  (:then
                    (bind _output_get (Variables|BPSetVariables|SetLastHitObstacle _asbuilding_actor))
                    (if (and (> (Health|GetMaxHealth _output_get :!class /Script/FortniteGame.BuildingActor) 0.0) (!= _output_get (Variables|HitInfo|GetLastHitNonDestroyedBuilding)))
                      (CallFunction|DamageBuilding _lasthitobstacle)
                      (if (<= (Health|GetHealth _lasthitobstacle :!class /Script/FortniteGame.BuildingActor) 0)
                        (Development|PrintString (Utilities|String|Append "BROKE AND CONTINUE:   " (Utilities|GetDisplayName (Variables|BPSetVariables|GetLastHitObstacle))))
                        (if (Variables|Default|GetReturnToSender)
                          (Default|HandleProjectileStopReturn) ; continue returning to player
                          (else
                            (Default|HandleProjectileStop) ; continue moving forward
                          ))
                        (else
                          ;;[1 0.4524 0.2474] Hit a destructible item, but did not destroy it so return to player
                          (bind _output_get_2 (Variables|HitInfo|SetLastHitNonDestroyedBuilding (Variables|BPSetVariables|GetLastHitObstacle))) ; Save this building so we can check it in ViableHit?\nThis fixes buildings taking multiple hits when the saber is "bouncing" off\nof them or passing through them. 
                          (Development|PrintString (Utilities|String|Append "Damaged and STOP:   " (Utilities|GetDisplayName _output_get_2)))))
                      (else
                        (Development|PrintString (Utilities|String|Append "ViableHit returned false, ignoring hit on " (Utilities|GetDisplayName (Variables|BPSetVariables|GetLastHitObstacle)))))))
                  (:CastFailed
                    (Development|PrintString (Utilities|String|Append "Hit a destructible, non-building actor " _returnvalue))))
                (Transformation|SetActorLocation (+ (Transformation|GetActorLocation) (* (Math|Vector|GetUnitDirection(Vector) (Transformation|GetActorLocation) (Transformation|GetActorLocation (Variables|Owning|GetOwningPlayerPawn))) 1e+02)) false true) ; If we hit something solid then we need to move projectile \ntowards player or it will be stuck
                (if (Variables|Default|GetReturnToSender)
                  (Collision|IgnoreActorwhenMoving (Variables|Collision|GetCapsuleComponent) _hitactor true) ; If we didn't destroy the wall, pass through it
                  (Default|HandleProjectileStopReturn))))
            (else
              ;;[1 0.6685 0.2208] Hit indestructible
              ;; If already returning, keep going. If not, return
              (Development|PrintString (Utilities|String|Append "Hit Indestructible: " _returnvalue))
              (if (Variables|Default|GetReturnToSender)
                (Collision|IgnoreActorwhenMoving (Variables|Collision|GetCapsuleComponent) _hitactor true)
                (Default|HandleProjectileStopReturn)
                (else
                  (CallFunction|GCFeedback (Variables|GC|GetGC_EnvImpact) (Transformation|GetActorLocation self) (Transformation|GetActorForwardVector self) _hitactor))))))))))

(event Custom|UpdateWaterLocation)

(event Custom|SetAutoReturnTimerLower
  ;[0.01036 0.9901 1] Clear existing auto return timer and set it to a lower time
  (Utilities|Time|ClearandInvalidateTimerbyHandle (Variables|TimerHandles|GetAuto%20Return%20Timer%20Handle))
  (bind _returnvalue (Utilities|Time|SetTimerbyEvent (delegate Custom|ShorterTimer) (ScalableFloat|GetValueatLevel (Variables|Bounce|GetRow_BounceShortenFlightTimer))))
  (Variables|TimerHandles|SetAuto%20Return%20Timer%20Handle _returnvalue))

(event Custom|ShorterTimer
  ;[0.01036 0.9901 1] Clear existing auto return timer and set it to a lower time
  (Networking|SwitchHasAuthority
    (:Authority
      (if (not (Variables|Default|GetIsReturning))
        (Variables|Default|SetbReturnTimerExpired true)))))

(event OnEnterWaterEvent(WaterInteractionComponent)
  (params (WaterBody :type Fort%20Water%20Body%20Actor%20Object%20Reference) (WaterInteractionComponent :type Fort%20Water%20Interaction%20Component%20Object%20Reference) (bIsFirstBody :type Boolean)))

(event EventEndPlay
  (params (EndPlayReason :type EEndPlayReason%20Enum :default "Destroyed"))
  (bind _physical_wash_force (Variables|Default|GetPhysical%20Wash%20Force))
  (|CallParentFunction EndPlayReason)
  (Utilities|IsValid _physical_wash_force
    (:Is%20Valid
      (Actor|DestroyActor _physical_wash_force)
      (Utilities|Time|ClearandInvalidateTimerbyHandle (Variables|TimerHandles|GetUpdateDustLocationTimer)))) ; Checks if Rotor Wash Actor is Valid
)

(event Custom|SplashFX
  (params (WaterBody :type Fort%20Water%20Body%20Actor%20Object%20Reference) (WaterInteractionComponent :type Fort%20Water%20Interaction%20Component%20Object%20Reference) (bIsFirstBody :type Boolean))
  ;[0.05432 0.3611 1] Water Interactions
  (bind _outwatersurfaceinfo (WaterBody|GetWaterSurfaceInfoStruct WaterBody (Transformation|GetActorLocation) true false false))
  (bind _returnvalue (Effects|Components|ParticleSystem|SpawnFXAtLocation (Variables|VFX|GetNS%20Water%20Splash%20VFX) (Utilities|Struct|BreakWaterSurfaceInfo _outwatersurfaceinfo) (Math|Rotator|MakeRotfromX (.WaterPlaneNormal (Utilities|Struct|BreakWaterSurfaceInfo _outwatersurfaceinfo)))))
  (Effects|Components|ParticleSystem|SetFloatParameter _returnvalue "WaterDepth" (.WaterDepth (Utilities|Struct|BreakWaterSurfaceInfo _outwatersurfaceinfo)))
  (Effects|Components|ParticleSystem|SetFloatParameter _returnvalue "WaterBodyIndex" (Math|Conversions|ToFloat(Integer) (.WaterBodyIdx (Utilities|Struct|BreakWaterSurfaceInfo _outwatersurfaceinfo))))
  (Audio|SpawnSoundatLocation 0 (Transformation|GetActorLocation)) ; Water Splash SFX
)

(event OnExitWaterEvent(WaterInteractionComponent)
  (params (WaterBody :type Fort%20Water%20Body%20Actor%20Object%20Reference) (WaterInteractionComponent :type Fort%20Water%20Interaction%20Component%20Object%20Reference) (bIsLastBody :type Boolean))
  ;[0.05432 0.3611 1] Water Interactions
  (bind _outwatersurfaceinfo (WaterBody|GetWaterSurfaceInfoStruct WaterBody (Transformation|GetActorLocation) true false false))
  (bind _returnvalue (Effects|Components|ParticleSystem|SpawnFXAtLocation (Variables|VFX|GetNS%20Water%20Splash%20VFX) (Utilities|Struct|BreakWaterSurfaceInfo _outwatersurfaceinfo) (Math|Rotator|MakeRotfromX (.WaterPlaneNormal (Utilities|Struct|BreakWaterSurfaceInfo _outwatersurfaceinfo)))))
  (Effects|Components|ParticleSystem|SetFloatParameter _returnvalue "WaterDepth" (.WaterDepth (Utilities|Struct|BreakWaterSurfaceInfo _outwatersurfaceinfo)))
  (Effects|Components|ParticleSystem|SetFloatParameter _returnvalue "WaterBodyIndex" (Math|Conversions|ToFloat(Integer) (.WaterBodyIdx (Utilities|Struct|BreakWaterSurfaceInfo _outwatersurfaceinfo)))))

(event EventBeginPlay
  (bind _projectilemovementcomponent (Variables|Movement|GetProjectileMovementComponent))
  ;[1 1 1] Begin Play.  Return if nothing is hit after a delay
  (|CallParentFunction)
  (Utilities|FlowControl|Sequence
    (:then_0
      (bind _asfort_player_pawn (Utilities|Casting|CastToFortPlayerPawn (Variables|Actor|GetInstigator)))
      (bind _output_get (Variables|Owning|SetOwningPlayerPawn _asfort_player_pawn))
      (GameplayEffects|ApplyGameplayEffectToTarget :self (Ability|GetAbilitySystemComponent _output_get) :GameplayEffectClass "/AncientWeapon/Gameplay/GameplayEffects/GE_AncientWeapon_Thrown_BlockingFire.GE_AncientWeapon_Thrown_BlockingFire_C" :Target (Ability|GetAbilitySystemComponent _output_get) :Level 0.0) ; hides shield on arm mesh.  Needs to go here so it never gets applied until proj has spawned
      (bind _asb_space_clique_weapon (Utilities|Casting|CastToB_SpaceClique_Weapon (Variables|/Script/FortniteGame.FortPawn|GetCurrentWeapon (Variables|Owning|GetOwningPlayerPawn))))
      (Variables|Default|SetB%20Space%20Clique%20Weapon _asb_space_clique_weapon))
    (:then_1
      (CallFunction|SetHomingVars)
      (Networking|SwitchHasAuthority
        (:Authority
          (bind _b_space_clique_weapon (Variables|Default|GetB%20Space%20Clique%20Weapon)
            (:then
              (Class|BSpaceCliqueWeapon|SetOrbProjectileEvent _b_space_clique_weapon 0))
            (:else)))))
    (:then_2
      (Variables|Default|SetCurrentNumberOfBounces 0)
      ;;[1 0 0.02864] TODO: remove delay here if code fix happens.Delay here is intentional to allow this to overlap targets that are the homing target. This is a BP hack for an issue Paxton is looking into on homing projectile and first frame behavior.
      (Utilities|FlowControl|DelayUntilNextTick)
      (Variables|/Script/Engine.ProjectileMovementComponent|SetInitialSpeed (ScalableFloat|GetValueatLevel (Variables|BPSetVariables|GetRow_ThrowInitialSpeed)) _projectilemovementcomponent)
      (Variables|/Script/Engine.ProjectileMovementComponent|SetMaxSpeed (ScalableFloat|GetValueatLevel (Variables|BPSetVariables|GetRow_ThrowMaxSpeed)) _projectilemovementcomponent))
    (:then_3
      (CallFunction|SetinAirLoopSound))
    (:then_4
      (Utilities|Array|AddUnique (Variables|Targeting|GetActorsToIgnoreForTargeting) (Variables|Owning|GetOwningPlayerPawn))
      ;;[0 0.334 1] The projectile starts to return automatically after this timer expires, unless it meets other conditions first
      (bind _returnvalue (Utilities|Time|SetTimerbyEvent (delegate Custom|TriggerReturn) (ScalableFloat|GetValueatLevel (Variables|BPSetVariables|GetRow_AutoReturnTime))))
      (Variables|TimerHandles|SetAuto%20Return%20Timer%20Handle _returnvalue)
      (bind _returnvalue_2 (Utilities|Time|SetTimerbyEvent (delegate Custom|SlowDownCurve) (Math|Float|Clamp(Float) (- (ScalableFloat|GetValueatLevel (Variables|BPSetVariables|GetRow_AutoReturnTime)) 0.75) 0.5 (- (ScalableFloat|GetValueatLevel (Variables|BPSetVariables|GetRow_AutoReturnTime)) 0.75))))
      (Variables|TimerHandles|SetSlowDownTimerHandle _returnvalue_2))
    (:then_5
      (Variables|Default|SetPreviousVelocity (Variables|/Script/Engine.MovementComponent|GetVelocity (Variables|Movement|GetProjectileMovementComponent)))
      (Variables|Default|SetStartLocation (Transformation|GetActorLocation self))
      ;[1 1 1] If player dies, destroy projectile
      (EventDispatchers|BindEventtoOnDied (Variables|Owning|GetOwningPlayerPawn) (delegate Custom|OnDied) :!class /Script/FortniteGame.FortPawn)
      (Game|BindEventtoOnDestroyed (EventDispatchers|CreateEvent :!event Custom|ProjectileDestroyed))
      (Water|BindEventtoOnEnterWaterEvent (Variables|Water|GetWaterInteractionComponent) (EventDispatchers|CreateEvent :!event Custom|SplashFX))
      (Variables|Sounds|SetWaterFuseSound (Variables|Audio|GetHitWaterSoundCue)))
    (:then_6
      (Utilities|FlowControl|Sequence
        (:then_0
          ;[1 1 1] Comment
          (bind _returnvalue_3 (Utilities|Time|SetTimerbyEvent (delegate Custom|UpdateDustLocation) 0.15 true))
          (Variables|TimerHandles|SetUpdateDustLocationTimer _returnvalue_3))
        (:then_1
          ;[1 1 1] FX - PhysicsForce
          (Utilities|EffectsQualityCheck 0
            (:Cinematic
              _)
            (:Epic
              _))
          (Utilities|IsValid (Variables|Default|GetPhysical%20Wash%20Force)
            (:Is%20Not%20Valid
              (bind _returnvalue_4 (Game|SpawnActorfromClass "/Game/Effects/Niagara/Blueprints/PhsyicsFields/FN_WashForce.FN_WashForce_C" (Transformation|GetActorTransform self) "Undefined" "MultiplyWithRoot" 0 5e+02 (Variables|Default|GetRotorWashForceStrength) -999.0 0))
              (bind _output_get_2 (Variables|Default|SetPhysical%20Wash%20Force _returnvalue_4))
              (Transformation|AttachActorToComponent _output_get_2 (Variables|Default|GetSkeletalMesh) "None" "SnapToTarget" "KeepRelative" "KeepRelative" true))) ; Don't spawn the Rotor Wash multiple times
        )))))

(event Custom|Recall%20ReturnToOwner
  (bind _owningplayerpawn (Variables|Owning|GetOwningPlayerPawn))
  ;[1 1 1] Recall & Return To Owner
  (Networking|SwitchHasAuthority
    (:Authority
      (Utilities|Time|ClearandInvalidateTimerbyHandle (Variables|TimerHandles|GetReturn%20Arc%20Timer%20Handle))
      (Utilities|Time|ClearandInvalidateTimerbyHandle (Variables|TimerHandles|GetAuto%20Return%20Timer%20Handle))
      ;;[0.06679 1 0.1159] Recall, do explosion targeting here
      (if (ScalableFloat|GetValueasBool (Variables|Recall|GetAllowRecallWhileReturning))
        _
        (elif (Variables|Default|GetIsReturning)
          _))
      (CallFunction|DoRadialTargeting)
      (CallFunction|GCFeedback (Variables|GC|GetProx%20Explosion%20Gameplay%20Cue%20Tag) (Transformation|GetActorLocation self) (Transformation|GetActorForwardVector self))
      (Development|PrintString "Return to Owner Triggered")
      (Default|HandleProjectileStopReturn)
      (Utilities|IsValid _owningplayerpawn
        (:Is%20Valid
          (bind _distance (Class|MissionBlueprintFunctionLibrary|DistanceBetweenTwoVectors (Transformation|GetActorLocation) (Transformation|GetActorLocation _owningplayerpawn)))
          (if (>= _distance (ScalableFloat|GetValueatLevel (Variables|HomingValues|GetRow_ReturnHomingLockTargetDistanceThreshold)))
            (Variables|Default|SetIsReturning true)
            (CallFunction|TurnonRecallFX)
            (Variables|Default|SetReturnToSender true) ; Set Return Things
            (CallFunction|PrjReturn true) ; Manually Recalled Return Case
            (Networking|ForceNetUpdate)
            (else
              _)))
        (:Is%20Not%20Valid
          _))
      (CallFunction|KillProjectile))))

(event Custom|SlowDownCurve
  (Networking|SwitchHasAuthority
    (:Authority
      (if (not (Variables|Default|GetIsReturning))
        (bind (_direction _newtrack_0) (Timeline|SlowdownCurveTimeline PlayFromStart)
          (:Update
            (bind _output_get (Variables|/Script/Engine.ProjectileMovementComponent|SetMaxSpeed (* _newtrack_0 (Variables|/Script/Engine.ProjectileMovementComponent|GetMaxSpeed (Variables|Movement|GetProjectileMovementComponent))) (Variables|Movement|GetProjectileMovementComponent)))))))))

(event Custom|TriggerReturn
  ;[1 1 1] Begin Play.  Return if nothing is hit after a delay
  ;;[0 0.334 1] The projectile starts to return automatically after this timer expires, unless it meets other conditions first
  (Networking|SwitchHasAuthority
    (:Authority
      (if (not (Variables|Default|GetIsReturning))
        (Variables|Default|SetbReturnTimerExpired true)
        (CallFunction|PrintReturnInfo)))))

(event Custom|TurnOnRecallFX
  (bind _niagara (Variables|Default|GetNiagara))
  ;[1 1 1] Turn on return FX and sounds
  (Utilities|FlowControl|DoOnce false
    (:Completed
      (Development|PrintString "Recall FX")
      (Default|PlayRecallSound)
      (Niagara|SetNiagaraVariable(Bool) _niagara "ImpactObject" true)
      (Niagara|SetNiagaraVariable(Bool) _niagara "IsRecall" true)
      (CallFunction|ExecuteAudioParameter "OnExplosion"))))

(event Custom|OnDied
  (params (DamagedActor :type Actor%20Object%20Reference) (Damage :type Float%20%28single-precision%29) (InstigatedBy :type Controller%20Object%20Reference) (DamageCauser :type Actor%20Object%20Reference) (HitLocation :type Vector) (FHitComponent :type Primitive%20Component%20Object%20Reference) (BoneName :type Name) (Momentum :type Vector))
  ;[1 1 1] If player dies, destroy projectile
  (CallFunction|KillProjectile))

(event Custom|CallReturnMovement
  ;[1 1 1] Call Return Movement
  (Default|ResumeMovement false) ; Do not reset the velocity when returning
  (CallFunction|PlayImpactFeedback))

(event Custom|UpdateDustLocation
  (bind _niagara (Variables|Default|GetNiagara))
  ;[1 1 1] Comment
  (bind (_outhit _returnvalue) (Collision|LineTraceByChannel :Start (+ (Transformation|GetActorLocation) "0.000000,0.000000,100.000000") :End (+ (Transformation|GetActorLocation) "0.000000,0.000000,-500.000000") :TraceChannel "TraceTypeQuery1" :bTraceComplex false :DrawDebugType "None" :bIgnoreSelf true :TraceColor "(R=1.000000,G=0.000000,B=0.000000,A=1.000000)" :TraceHitColor "(R=0.000000,G=1.000000,B=0.000000,A=1.000000)" :DrawTime 2.0))
  (Niagara|SetNiagaraVariable(Bool) _niagara "GroundDust" _returnvalue)
  (Niagara|SetNiagaraVariable(Position) _niagara "DustLocation" (.Location (Collision|BreakHitResult _outhit))))

(event Custom|PlayImpactFeedback
  ;[1 1 1] Trigger controller feedback
  (Networking|CanExecuteCosmeticEvents
    (:True
      (bind _owningplayerpawn (Variables|Owning|GetOwningPlayerPawn)
        (:then
          (if (Pawn|IsLocallyControlled _owningplayerpawn)
            (ForceFeedback|SpawnForceFeedbackAttached "/Game/Weapons/ForceFeedback/FF_Generic_Impact_Near.FF_Generic_Impact_Near" (Variables|Transformation|GetRootComponent _owningplayerpawn))))
        (:else)))))

(event Custom|CallMovement
  ;[1 1 1] Call Outgoing Movement
  (Default|ResumeMovement true)
  (CallFunction|PlayImpactFeedback))

(event EventOnTouched
  (params (OtherActor :type Actor%20Object%20Reference) (OtherComp :type Primitive%20Component%20Object%20Reference) (HitResult :type Hit%20Result%20Structure%20%28by%20ref%29) (bIsOverlap :type Boolean))
  ;[0.4662 1 0.9235] Event OnTouched - If overlap, call HandleOverlap to potentially damage a character
  ; If not an overlap, cache the velocity for use in the on stop event
  (if bIsOverlap
    (Variables|HitInfo|SetHitResult HitResult)
    (bind _ascharacter (Utilities|Casting|CastToCharacter OtherActor))
    (CallFunction|HandleOverlap _ascharacter)
    (if (and (ScalableFloat|GetValueasBool (Variables|BPSetVariables|GetRow_AllowReturnOnPawnHit)) (not (Variables|Default|GetIsReturning)))
      (CallFunction|TravelReturnArcAfterHit true)
      (elif (and (ScalableFloat|GetValueasBool (Variables|BPSetVariables|GetRow_ForceReturnAfterPawnHit)) (not (Variables|Default|GetIsReturning)))
        (Variables|Default|SetIsReturning true)
        (Variables|Default|SetReturnToSender true) ; Set Return Things
        (CallFunction|PrjReturn false false true)))
    (else
      (Variables|Default|SetPreviousVelocity (Transformation|GetVelocity)) ; Set Previous Velocity so on ResumeMovement we know the direction to go.\nSetting it here because this gets triggered before the stop event
    )))

(event Custom|DamageBuilding
  (params (Building%20Actor :type Building%20Actor%20Object%20Reference))
  ;[1 1 1] Damage building
  (CallFunction|GCFeedback (Variables|GC|GetGC_EnvImpact) (Transformation|GetActorLocation self) (Transformation|GetActorForwardVector self) Building%20Actor)
  (bind _bdidactorhavetags (Class|AthenaQuestFunctionLibrary|AthenaCheckActorHasTags Building%20Actor (Variables|Default|GetInstaKillTags)))
  (bind _hit_effect_context (CallFunction|SetHitEffectContext))
  (Ability|EffectContext|AddHitResult _hit_effect_context (Variables|HitInfo|GetHitResult))
  (GameplayEffects|ApplyGameplayEffectToTarget (Ability|GetAbilitySystemComponent (Variables|Owning|GetOwningPlayerPawn)) (select _bdidactorhavetags (Variables|GE&Damage|GetProjectileEnvDamage_InstaKill) (Variables|GE&Damage|GetProjectileEnvDamage)) (Ability|GetAbilitySystemComponent Building%20Actor) 0.0 (Variables|HitInfo|GetHitEffectContext)))

(event Custom|HandleOverlap
  (params (OverlapActor :type Character%20Object%20Reference))
  (bind _projectilemovementcomponent (Variables|Movement|GetProjectileMovementComponent))
  ;[0.5948 0.9165 1] Overlap - Potentially damage character
  (Niagara|SetNiagaraVariable(Bool) (Variables|Default|GetNiagara) "ImpactPlayer" true)
  (Networking|SwitchHasAuthority
    (:Authority
      ;;[1 1 1] ignore the hit if player was recently hit
      (bind _isimmune (Default|CheckImmunityStatus OverlapActor))
      (if (not _isimmune)
        (if (not (Math|Boolean|ORBoolean (Utilities|Equal(Object) (Variables|Actor|GetInstigator) OverlapActor) (Team|OnSameTeam OverlapActor (Variables|Owning|GetOwningPlayerPawn)) (Utilities|Array|ContainsItem (Actor|GetAttachedActors OverlapActor true false) (Variables|Actor|GetInstigator))))
          (Default|EvaluateEndingHomingBehavior OverlapActor) ; stop homing if we reached the target
          (CallFunction|DamageHitActor OverlapActor)
          (Utilities|Map|Add (Variables|HitInfo|GetRecentlyHit) OverlapActor (Fortnite|Sync|Time|GetServerWorldTimeSeconds))
          ;[1 0.9332 0.1389] Stop Homing on Hit Pawn and continue flight., TODO: From Evaluate Ending Homing func, current BP doesn't have all the lobster realm homing stuff, so mimicing this specificly here
          ;;[1 1 1] Don't reset homing if we're returing, as we are still homing  in on the owning player pawn
          (if (not (Variables|Default|GetIsReturning))
            (Projectile|SetHomingData _projectilemovementcomponent (Utilities|Struct|MakeProjectileHomingData "None" (ScalableFloat|GetValueatLevel (Variables|HomingValues|GetRow_HomingMinTurnSpeed)) (ScalableFloat|GetValueatLevel (Variables|HomingValues|GetRow_HomingMaxTurnSpeed)) (ScalableFloat|GetValueatLevel (Variables|HomingValues|GetRow_HomingTurnRampTime)) (ScalableFloat|GetValueatLevel (Variables|HomingValues|GetRow_HomingEffectTriggerRange))) :!class /Script/FortniteGame.FortProjectileMovementComponent) ; reset the homing parameters to stop following the target
            (Components|Movement|SetUpdatedComponent _projectilemovementcomponent (Variables|Collision|GetCapsuleComponent))))
        (else
          (Development|PrintString "Immune from Attack, ignoring"))))))

(event Custom|BounceAndMirrorVectorNormal
  (bind _returnvalue (CallFunction|CheckDistanetoPopUpandReturnEarly))
  ;[0 1 0.1304] Bounce
  (Networking|SwitchHasAuthority
    (:Authority
      (if (>= (Variables|Default|GetCurrentNumberOfBounces) (ScalableFloat|GetValueasInteger (Variables|Bounce|GetRow_ReturnOnThisBounce)))
        ;;[0.09837 0.3942 1] Immediately return after 2nd bounce, don't bounce for the full lifetime. If short hit, use that, otherwise force return
        (Variables|Default|SetIsReturning true)
        (Variables|Default|SetReturnToSender true) ; Set Return Things
        (CallFunction|StopShortHitPopUpTimeline)
        (CallFunction|PrjReturn false false true)
        (else
          (Variables|Default|SetLast%20HIt%20Location (Variables|HitInfo|GetBoomerangHitLoc))
          (bind _output_get (Variables|HitInfo|SetVector_AngleOfIncidence (* (Math|Vector|MirrorVectorbyNormal (Variables|/Script/Engine.MovementComponent|GetVelocity (Variables|Movement|GetProjectileMovementComponent)) (Variables|HitInfo|GetBoomerangHitNormal)) (ScalableFloat|GetValueatLevel (Variables|Bounce|GetRow_BouncevelocityMultiplier)))))
          (Variables|HitInfo|Setdot (Math|Vector|DotProduct (Math|Vector|Normalize (- (Variables|HitInfo|GetBoomerangHitLoc) (Transformation|GetActorLocation (Variables|Owning|GetOwningPlayerPawn)))) (Math|Vector|Normalize _output_get)))
          (if _returnvalue
            (CallFunction|ShortHitEvent)
            (elif (Math|Boolean|ANDBoolean (> (Utilities|Time|GetTimerRemainingTimebyHandle (Variables|TimerHandles|GetAuto%20Return%20Timer%20Handle)) (ScalableFloat|GetValueatLevel (Variables|Bounce|GetRow_BounceShortenFlightTimerCheckTime))) (CallFunction|BIsGroundHitCloseDistance) (not _returnvalue))
              (CallFunction|SetAutoReturnTimerLower)))
          (Development|PrintString (Utilities|String|Append "Dot product of hit: " (Utilities|String|ToString(Float) (Variables|HitInfo|Getdot))) (Variables|Default|GetbDebug) (Variables|Default|GetbDebug))
          ;;[1 0.9822 0] Hit a wall, at specific angles, force return bounce back to player
          (if (and (< (Variables|HitInfo|Getdot) (ScalableFloat|GetValueatLevel (Variables|Bounce|GetRow_BounceDotCheckForForceReturn))) (not (> (.z (Variables|HitInfo|GetBoomerangHitNormal)) (ScalableFloat|GetValueatLevel (Variables|Bounce|GetRow_BounceHitNormalZCheckForPopUp)))))
            _
            (elif (Variables|Default|GetbReturnTimerExpired)
              _))
          (Default|ResumeMovement true)
          (Variables|ReturnArc|SetHas%20Flight%20Angle%20Changed%20Based%20on%20Impact true)
          (Variables|Default|SetIsReturning true)
          (Variables|Default|SetReturnToSender true) ; Set Return Things
          (CallFunction|PrjReturn false false true) ; Force Return Case
        )))))

(event Custom|DamageHitActor
  (params (Actor :type Actor%20Object%20Reference))
  (bind _returnvalue (Game|GetInstigator))
  (bind _returnvalue_2 (Ability|GetAbilitySystemComponent Actor))
  (bind _returnvalue_3 (Ability|GetAbilitySystemComponent _returnvalue))
  ;[0.4688 0.3006 0.1296] Damage Hit Actor
  (Utilities|IsValid Actor
    (:Is%20Valid
      (bind _b_space_clique_weapon (Variables|Default|GetB%20Space%20Clique%20Weapon)
        (:then
          (Class|BSpaceCliqueWeapon|HandleProjectileHit _b_space_clique_weapon))
        (:else
          _))
      (CallFunction|GCFeedback (Variables|GC|GetGC_PlayerImpact) (Transformation|GetActorLocation self) (Transformation|GetActorForwardVector self) Actor)
      (Utilities|IsValid _returnvalue_2
        (:Is%20Valid
          (Utilities|IsValid _returnvalue
            (:Is%20Valid
              (Utilities|IsValid _returnvalue_3
                (:Is%20Valid
                  (bind _hit_effect_context (CallFunction|SetHitEffectContext))
                  (Ability|EffectContext|AddHitResult _hit_effect_context (Variables|HitInfo|GetHitResult))
                  ;;[0.6727 0.4651 1] Damage Players & Vehicles, both handled in this GE. Select based on if returning & select if we should ignore shields or not
                  (GameplayEffects|ApplyGameplayEffectToTarget _returnvalue_3 (select (Variables|Default|GetIsReturning) (select (ScalableFloat|GetValueasBool (Variables|GE&Damage|GetRow_DoDirectDamageOnReturn)) (Variables|GE&Damage|GetGE_OnReturn_PlayerAndVehicle_DirectDamage) (Variables|GE&Damage|GetGE_OnReturn_PlayerAndVehicleDamage)) (Variables|GE&Damage|GetGE_PlayerAndVehicleDamage)) _returnvalue_2 0.0 (Variables|HitInfo|GetHitEffectContext)))))))))))

(event Custom|ShortHitEvent
  ;[0.06779 1 0.04894] Handle Close Hit cases so we have a better idea of what the projectile is doing
  ;;[0.9692 1 0.152] get point up and slightly away from hit, bounce and spin, then return
  (bind _outvector (Default|FindShortHitLocation))
  (Variables|ReturnArc|SetReturn%20Point%20Vector _outvector)
  (Variables|Targeting|SetHomingTarget 0)
  (Default|ResumeMovement false) ; Do not reset the velocity when returning
  ;;[1 0.8876 0.07081] Create a velocity to move in the pop up line we set in Find Short Hit Location, then scale speed to 0 to spin in air for a bit before return
  (Projectile|SetVelocity (Variables|Movement|GetProjectileMovementComponent) (- (Variables|ReturnArc|GetReturn%20Point%20Vector) (Variables|HitInfo|GetBoomerangHitLoc)))
  (CallFunction|ScaleMovementSpeedDown))

(event Custom|LaunchHitActor
  (params (Actor :type Actor%20Object%20Reference))
  ;[1 1 1] Launch Actor
  (bind _ascharacter (Utilities|Casting|CastToCharacter Actor)
    (:then
      (bind _returnvalue (GameplayTags|HasMatchingGameplayTag(Message) _ascharacter "(TagName=\"Gameplay.Status.KnockbackImmunity\")"))
      (if (not _returnvalue)
        (Character|LaunchCharacter _ascharacter (Default|CalculateKnockbackVelocity _ascharacter) true true)))
    (:CastFailed
      (bind _asfort_vehicle_interface (Utilities|Casting|CastToFortVehicleInterface Actor))
      (Audio|BroadcastAbilityImpactatLocation 0 (Utilities|ToObject(Interface) _asfort_vehicle_interface) (Default|CalculateKnockbackVelocity _ascharacter) 777.0) ; This launches vehicles (named weird for anti-cheat reasons)\nPosition = velocity to launch it at\nRadius must equal 777\nGates knows more about it
    )))

(event Custom|GCFeedback
  (params (GameplayCueTag :type Gameplay%20Tag%20Structure) (Location :type Vector) (Normal :type Vector) (SourceObject :type Object%20Reference))
  ;[1 0.8018 0] GAMEPLAY CUE FEEDBACK
  (GameplayCue|ExecuteGameplayCueOnActor(Burst) self GameplayCueTag (Ability|GameplayCue|MakeGameplayCueParameters :NormalizedMagnitude 0.0 :RawMagnitude 0.0 :Location Location :Normal Normal :Instigator self :EffectCauser self :SourceObject SourceObject :GameplayEffectLevel 1 :AbilityLevel 1 :bReplicateLocationWhenUsingMinimalRepProxy false)))

(event Custom|ScaleMovementSpeedDown
  (bind _projectilemovementcomponent (Variables|Movement|GetProjectileMovementComponent))
  ;[0.06779 1 0.04894] Handle Close Hit cases so we have a better idea of what the projectile is doing
  ;;[0 0.2362 1] Scale movement speed to a stop
  (bind (_direction _speedratio) (Timeline|Timeline_1 PlayFromStart)
    (:Update
      (Variables|/Script/Engine.ProjectileMovementComponent|SetMaxSpeed (Math|Float|Lerp (ScalableFloat|GetValueatLevel (Variables|Bounce|GetRow_BouncePopUpInitialSpeed)) (ScalableFloat|GetValueatLevel (Variables|Bounce|GetRow_BouncePopUpEndSpeed)) _speedratio) _projectilemovementcomponent))
    (:Finished
      (Variables|/Script/Engine.MovementComponent|SetVelocity "0, 0, 0" _projectilemovementcomponent)
      ;;[0.03692 0.179 1] Once we stop, start the return. If the player is close enough, just kill the prj and catch it
      (Utilities|Time|SetTimerbyEvent (delegate Custom|PostSlowSpeedScale) (ScalableFloat|GetValueatLevel (Variables|Bounce|GetRow_BounceTimeBeforeReturnAfterPopUp))))))

(event Custom|PostSlowSpeedScale
  (bind _owningplayerpawn (Variables|Owning|GetOwningPlayerPawn))
  ;[0.06779 1 0.04894] Handle Close Hit cases so we have a better idea of what the projectile is doing
  ;;[0.03692 0.179 1] Once we stop, start the return. If the player is close enough, just kill the prj and catch it
  (Utilities|IsValid _owningplayerpawn
    (:Is%20Valid
      (bind _distance (Class|MissionBlueprintFunctionLibrary|DistanceBetweenTwoVectors (Transformation|GetActorLocation) (Transformation|GetActorLocation _owningplayerpawn)))
      (if (>= _distance (ScalableFloat|GetValueatLevel (Variables|HomingValues|GetRow_ReturnHomingLockTargetDistanceThreshold)))
        (Variables|Default|SetIsReturning true)
        (CallFunction|TurnonRecallFX)
        (Variables|Default|SetReturnToSender true) ; Set Return Things
        (CallFunction|PrjReturn false true) ; Short Hit Return Case
        (Networking|ForceNetUpdate)
        (else
          _)))
    (:Is%20Not%20Valid
      _))
  (CallFunction|KillProjectile))

(event Custom|Stop%20Short%20Hit%20Pop%20Up%20Timeline
  (bind _projectilemovementcomponent (Variables|Movement|GetProjectileMovementComponent))
  ;[0.06779 1 0.04894] Handle Close Hit cases so we have a better idea of what the projectile is doing
  ;;[0 0.2362 1] Scale movement speed to a stop
  (Timeline|Timeline_1 Stop))

(event Custom|KillProjectile
  (bind _owningplayerpawn (Variables|Owning|GetOwningPlayerPawn))
  (bind _owningplayerpawn_2 (Variables|Owning|GetOwningPlayerPawn))
  ;[1 0.7007 0.4203] Kill Projectile (Called from OnDied)
  (Utilities|FlowControl|DoOnce false
    (:Completed
      (Utilities|IsValid _owningplayerpawn_2
        (:Is%20Valid
          (Utilities|IsValid (Ability|GetAbilitySystemComponent _owningplayerpawn_2)
            (:Is%20Valid
              (GameplayCue|RemoveGameplayCueOnActor(Looping) _owningplayerpawn (Variables|GC|GetLooping%20Gameplay%20Cue%20Tag) (Ability|GameplayCue|MakeGameplayCueParameters 0.0))
              (Ability|SendGameplayEventtoActor _owningplayerpawn (Variables|Event|GetCatchEventTag) (Variables|Default|GetCatchEventPayload)) ; Trigger Catch GA
            )
            (:Is%20Not%20Valid
              _)))
        (:Is%20Not%20Valid
          _))
      (SoundIndicator|StopSoundIndicatorLooping (Variables|Default|GetFortSoundIndicator))
      (CallFunction|ExecuteAudioParameter "OnStop")
      (bind _recallaudio (Variables|Default|GetRecallAudio)
        (:then
          (Audio|Components|Audio|FadeOut _recallaudio 0.35 0.0 "Linear" :!class /Script/Engine.AudioComponent))
        (:else
          _))
      ;;[0.4181 0.05552 1] Store off vector for VFX on catch. Otherwise the prj is killed by the time we try to access it
      (bind _b_space_clique_weapon (Variables|Default|GetB%20Space%20Clique%20Weapon)
        (:then
          (Variables|/SpaceClique/Gameplay/B_SpaceClique_Weapon.B_SpaceClique_Weapon_C|SetProjectileLastFwdVectorForCatchGC (Transformation|GetActorForwardVector self) _b_space_clique_weapon))
        (:else
          _))
      (Actor|DestroyActor)
      (Networking|SwitchHasAuthority
        (:Authority
          ;;[1 0.03073 0.02078] TODO: ???
          (bind _b_space_clique_weapon_2 (Variables|Default|GetB%20Space%20Clique%20Weapon)
            (:then
              (Class|BSpaceCliqueWeapon|SetOrbProjectileEvent _b_space_clique_weapon_2 0))
            (:else)))))))

(event Custom|ProjectileDestroyed
  (params (DestroyedActor :type Actor%20Object%20Reference)))

(event Custom|Travel%20Return%20Arc%20Start
  ;[0.04985 1 0.1746] Return Arcs
  (Variables|Default|SetIsReturning true)
  (Variables|Default|SetReturnToSender true) ; Set Return Things
  (Variables|ReturnArc|SetReturn%20Point%20Vector (CallFunction|GetReturnArcActorLocation))
  (if (Variables|Audio|GetbIsMetasound)
    _)
  (CallFunction|ExecuteAudioParameter "OnTurnInAir")
  (CallFunction|SpeedUpEvent))

(event Custom|DoRadialTargeting
  ;[0.1437 0.2378 1] Charged ammo Targeting & Damage
  (bind (_asynctaskproxy _targetinghandle) (Class|AsyncActionPerformTargeting|PerformTargetingAsyncAction 0 (Variables|Targeting|GetRadialDamageTargeting) true)
    (:Targeted
      (Utilities|FlowControl|Sequence
        (:then_0
          ;;[1 1 1] Deal environment damage
          (bind (_asynctaskproxy_2 _targetinghandle_2) (Class|AsyncActionPerformTargeting|PerformFilteringAsyncAction 0 (Variables|Targeting|GetBuildingTargetFilter) true (TargetingSystem|TargetingResults|GetTargetingResultsActors "/Script/TargetingSystem.Default__TargetingSubsystem" _targetinghandle))
            (:Targeted
              (CallFunction|SetHitEffectContext)
              (bind _asynctaskproxy_3 (Ability|Tasks|ApplyGameplayEffectToTargetingHandleAsync (Default|GetInstigatorASC) _targetinghandle_2 (Variables|GE&Damage|GetChargeEnvDamageGE) 1 1 (Variables|HitInfo|GetHitEffectContext)))
              (Variables|Targeting|SetEnvDamageTask _asynctaskproxy_3))))
        (:then_1
          (CallFunction|CheckforonHitDamage (TargetingSystem|TargetingResults|GetTargetingResultsActors "/Script/TargetingSystem.Default__TargetingSubsystem" _targetinghandle)))))))

(event Custom|Travel%20Return%20Arc%20After%20Hit
  (params (After%20Hit :type Boolean))
  ;[0.04985 1 0.1746] Return Arcs
  (Variables|Default|SetIsReturning true)
  (Variables|Default|SetReturnToSender true) ; Set Return Things
  (Variables|ReturnArc|SetReturn%20Point%20Vector (Default|GetReturnArcAfterPawnHit))
  (if (Variables|Audio|GetbIsMetasound)
    _)
  (CallFunction|ExecuteAudioParameter "OnTurnInAir")
  (CallFunction|SpeedUpEvent))

(event Custom|ReturnArcSetHomingAndTimer
  ;[0.04985 1 0.1746] Return Arcs
  (bind _returnvalue (Utilities|Time|SetTimerbyEvent (delegate Custom|SetProjectileReturnAfterArc) (Math|Float|SelectFloat (ScalableFloat|GetValueatLevel (Variables|BPSetVariables|GetRow_AllowReturnArcTime_AfterPawnHit)) (ScalableFloat|GetValueatLevel (Variables|BPSetVariables|GetRow_AllowReturnArcTime)) (stale Custom|Travel%20Return%20Arc%20After%20Hit After%20Hit))))
  (Variables|TimerHandles|SetReturn%20Arc%20Timer%20Handle _returnvalue))

(event Custom|SpeedUpEvent
  (bind _owningplayerpawn (Variables|Owning|GetOwningPlayerPawn))
  ;[0.04985 1 0.1746] Return Arcs
  (Networking|SwitchHasAuthority
    (:Authority
      (if (Variables|Default|GetIsReturning)
        (Projectile|SetHomingTarget 0 self _owningplayerpawn 0 :!class /Script/FortniteGame.FortProjectileMovementInterface)
        (Variables|Targeting|SetHomingTarget _owningplayerpawn)
        (CallFunction|TurnonRecallFX)
        (bind (_direction _newtrack_0) (Timeline|SpeedUpTimeline PlayFromStart)
          (:Update
            (bind _output_get (Variables|/Script/Engine.ProjectileMovementComponent|SetMaxSpeed (* _newtrack_0 (ScalableFloat|GetValueatLevel (Variables|BPSetVariables|GetRow_ReturnEndVelocity))) (Variables|Movement|GetProjectileMovementComponent)))))))))

(event Custom|CheckForOnHitDamage
  (params (InTargets :type Array%20of%20Actor%20Object%20References))
  ;[0.1437 0.2378 1] Charged ammo Targeting & Damage
  ;;[1 1 1] Pawn/vehicle damage application
  (Variables|Default|SetTargets%20to%20Filter%20for%20Explosive%20Damage InTargets)
  (for _array_element InTargets
    (bind _asfort_player_pawn (Utilities|Casting|CastToFortPlayerPawn _array_element))
    (if (Utilities|Map|Contains (Variables|HitInfo|GetRecentlyHit) _asfort_player_pawn)
      (Utilities|Array|RemoveItem (Variables|Default|GetTargets%20to%20Filter%20for%20Explosive%20Damage) _asfort_player_pawn)))
  (bind (_asynctaskproxy _targetinghandle) (Class|AsyncActionPerformTargeting|PerformFilteringAsyncAction 0 (Variables|Targeting|GetDamageOnHitFilter) true (Variables|Default|GetTargets%20to%20Filter%20for%20Explosive%20Damage))
    (:Targeted
      ;;;[1 1 1] TODO: Split out extra vehicle damage?
      (CallFunction|SetHitEffectContext)
      (Ability|Tasks|ApplyGameplayEffectToHitResultsAsync (Default|GetInstigatorASC) (TargetingSystem|TargetingResults|GetTargetingResults "/Script/TargetingSystem.Default__TargetingSubsystem" _targetinghandle) (Variables|GE&Damage|GetGE_Charged_Explosive_Damage) 1 1 (Variables|HitInfo|GetHitEffectContext)))))

(event Custom|SetProjectileReturnAfterArc
  ;[0.04985 1 0.1746] Return Arcs
  (Networking|SwitchHasAuthority
    (:Authority
      (CallFunction|PrjReturn false false false (stale Custom|Travel%20Return%20Arc%20After%20Hit After%20Hit)) ; Pawn Hit Return Case
      (bind _homing_return_arc_decoy_actor (Variables|Default|GetHoming%20Return%20Arc%20Decoy%20Actor)
        (:then
          (Actor|DestroyActor _homing_return_arc_decoy_actor))
        (:else)))))

(event Custom|CheckForOnHitImpulse
  (params (InTargets :type Array%20of%20Actor%20Object%20References))
  ;[0.1437 0.2378 1] Charged ammo Targeting & Damage
  ;;[1 1 1] Impulse application
  (bind (_asynctaskproxy _targetinghandle) (Class|AsyncActionPerformTargeting|PerformFilteringAsyncAction 0 (Variables|Targeting|GetImpulseOnHitFilter) true InTargets)
    (:Targeted
      (bind _asynctaskproxy_2 (Ability|Tasks|ApplyGameplayEffectToTargetingHandleAsync (Default|GetInstigatorASC) _targetinghandle (Variables|GE&Damage|GetGE_Impulse)))
      (Variables|Targeting|SetImpulseApplyTask _asynctaskproxy_2))))

(event Custom|PrjReturn
  (params (WasRecalled :type Boolean) (ShortHit :type Boolean) (ForceReturn :type Boolean) (PawnHit :type Boolean))
  (bind _self self)
  (bind _self_2 self)
  (bind _owningplayerpawn (Variables|Owning|GetOwningPlayerPawn))
  (bind _bdebug (Variables|Default|GetbDebug))
  ;[0.0643 0.2771 1] Return event - bring boomerang back
  (Utilities|FlowControl|DoOnce false
    (:Completed
      (if _bdebug
        (Development|PrintString (Utilities|String|Append "Setting return speed: " (Utilities|String|ToString(Float) 0.0)) _bdebug _bdebug)
        (Rendering|Debug|DrawDebugSphere (Transformation|GetActorLocation) 1e+02 12 "(R=0.000000,G=0.931906,B=1.000000,A=1.000000)" 6e+01 1.0))
      (Networking|SwitchHasAuthority
        (:Authority
          (Utilities|IsValid (Variables|Owning|GetOwningPlayerPawn)
            (:Is%20Valid
              (Projectile|SetHomingTarget 0 _self_2 _owningplayerpawn 0 :!class /Script/FortniteGame.FortProjectileMovementInterface)
              (Variables|Targeting|SetHomingTarget _owningplayerpawn)
              (if true
                (Projectile|SetHomingTimeUntilMaxTurnSpeed 0 _self_2 0.0 :!class /Script/FortniteGame.FortProjectileMovementInterface)
                (Projectile|SetHomingTurnSpeedMin 0 _self_2 1e+04 :!class /Script/FortniteGame.FortProjectileMovementInterface)
                (Projectile|SetHomingTurnSpeedMax 0 _self_2 1e+04 :!class /Script/FortniteGame.FortProjectileMovementInterface)
                (Projectile|SetHomingLockTargetDistanceThreshold 0 _self_2 (ScalableFloat|GetValueatLevel (Variables|HomingValues|GetRow_ReturnHomingLockTargetDistanceThreshold)) :!class /Script/FortniteGame.FortProjectileMovementInterface)
                (Projectile|SetHomingTargetOffset _self_2 (Variables|Default|GetReturnHomingOffset) :!class /Script/FortniteGame.FortProjectileMovementInterface)
                (Variables|/Script/Engine.ProjectileMovementComponent|SetMaxSpeed (ScalableFloat|GetValueatLevel (Variables|BPSetVariables|GetRow_RecallSpeedMax)) (Variables|Movement|GetProjectileMovementComponent))
                (else
                  (Projectile|SetHomingTimeUntilMaxTurnSpeed 0 _self 0.5 :!class /Script/FortniteGame.FortProjectileMovementInterface)
                  (Projectile|SetHomingTurnSpeedMin 0 _self 1e+04 :!class /Script/FortniteGame.FortProjectileMovementInterface)
                  (Projectile|SetHomingTurnSpeedMax 0 _self 1e+04 :!class /Script/FortniteGame.FortProjectileMovementInterface)
                  (Projectile|SetHomingLockTargetDistanceThreshold 0 _self (ScalableFloat|GetValueatLevel (Variables|HomingValues|GetRow_ReturnHomingLockTargetDistanceThreshold)) :!class /Script/FortniteGame.FortProjectileMovementInterface)
                  (Projectile|SetHomingTargetOffset _self (Variables|Default|GetReturnHomingOffset) :!class /Script/FortniteGame.FortProjectileMovementInterface)
                  (if (Math|Boolean|ORBoolean ShortHit ForceReturn PawnHit)
                    (Variables|/Script/Engine.ProjectileMovementComponent|SetMaxSpeed (Math|Float|SelectFloat (ScalableFloat|GetValueatLevel (Variables|BPSetVariables|GetRow_ReturnMaxSpeedCloseDistOnForceReturn)) (Math|Float|SelectFloat (ScalableFloat|GetValueatLevel (Variables|BPSetVariables|GetRow_ReturnMaxSpeedMedDistOnForceReturn)) (ScalableFloat|GetValueatLevel (Variables|BPSetVariables|GetRow_ReturnEndVelocity)) (and ForceReturn (< (Math|Vector|VectorLength (- (Transformation|GetActorLocation) (Transformation|GetActorLocation (Variables|Owning|GetOwningPlayerPawn)))) (ScalableFloat|GetValueatLevel (Variables|BPSetVariables|GetRow_ReturnMedDistCheckOnForceReturn))))) (and (< (Math|Vector|VectorLength (- (Transformation|GetActorLocation) (Transformation|GetActorLocation (Variables|Owning|GetOwningPlayerPawn)))) (ScalableFloat|GetValueatLevel (Variables|BPSetVariables|GetRow_ReturnCloseDistCheckOnForceReturn))) (not PawnHit))) (Variables|Movement|GetProjectileMovementComponent))
                    (Projectile|SetHomingTimeUntilMaxTurnSpeed 0 _self 0.0 :!class /Script/FortniteGame.FortProjectileMovementInterface)
                    (Projectile|SetHomingTurnSpeedMax 0 _self 5e+04 :!class /Script/FortniteGame.FortProjectileMovementInterface)
                    (else
                      (Utilities|Time|SetTimerbyEvent (delegate Custom|Set%20Return%20Homing%20Move%20Speed%20Max%20After%20Delay) 0.5))))))))))))

(event Custom|ExecuteAudioParameter
  (params (ParamName :type Name :default "None"))
  ;[1 1 1] Audio parameter for metasound
  (if (Variables|Audio|GetbIsMetasound)
    _)
  (Audio|Parameter|ExecuteTriggerParameter (Variables|Default|GetAncientWeapon_Primary_Loop_Cue) ParamName))

(event Custom|Set%20Return%20Homing%20Move%20Speed%20Max%20After%20Delay
  (bind _self self)
  ;[0.0643 0.2771 1] Return event - bring boomerang back
  (Variables|/Script/Engine.ProjectileMovementComponent|SetMaxSpeed (Math|Float|SelectFloat (ScalableFloat|GetValueatLevel (Variables|BPSetVariables|GetRow_ReturnMaxSpeedCloseDistOnForceReturn)) (Math|Float|SelectFloat (ScalableFloat|GetValueatLevel (Variables|BPSetVariables|GetRow_ReturnMaxSpeedMedDistOnForceReturn)) (ScalableFloat|GetValueatLevel (Variables|BPSetVariables|GetRow_ReturnEndVelocity)) (and (stale Custom|PrjReturn ForceReturn) (< (Math|Vector|VectorLength (- (Transformation|GetActorLocation) (Transformation|GetActorLocation (Variables|Owning|GetOwningPlayerPawn)))) (ScalableFloat|GetValueatLevel (Variables|BPSetVariables|GetRow_ReturnMedDistCheckOnForceReturn))))) (and (< (Math|Vector|VectorLength (- (Transformation|GetActorLocation) (Transformation|GetActorLocation (Variables|Owning|GetOwningPlayerPawn)))) (ScalableFloat|GetValueatLevel (Variables|BPSetVariables|GetRow_ReturnCloseDistCheckOnForceReturn))) (not (stale Custom|PrjReturn PawnHit)))) (Variables|Movement|GetProjectileMovementComponent))
  (Projectile|SetHomingTimeUntilMaxTurnSpeed 0 _self 0.0 :!class /Script/FortniteGame.FortProjectileMovementInterface)
  (Projectile|SetHomingTurnSpeedMax 0 _self 5e+04 :!class /Script/FortniteGame.FortProjectileMovementInterface))

(event ProjectileMovement|EventOnReachedLockTargetDistanceThreshold
  ;[0.9987 1 0.9767] Trigger catch event
  (if (and (Variables|Default|GetIsReturning) (== (Variables|Owning|GetOwningPlayerPawn) (Variables|Targeting|GetHomingTarget)))
    (CallFunction|KillProjectile)
    (Actor|DestroyActor)))

(event EventDestroyed
  (bind _owningplayerpawn (Variables|Owning|GetOwningPlayerPawn))
  (bind _returnvalue (Ability|GetAbilitySystemComponent _owningplayerpawn))
  ;[1 1 1] Event Destroyed: Remove Activation Block GE
  (Tags|UnBlockAbilitiesWithTags (Utilities|Casting|CastToFortAbilitySystemComponent (Ability|GetAbilitySystemComponent (Variables|Owning|GetOwningPlayerPawn))) (Variables|Default|GetChangeEquipmentTag) false)
  ;;[1 0.2489 0] unhooked unhide call so anim can try to do this on their end
  (Utilities|IsValid (Variables|Owning|GetOwningPlayerPawn)
    (:Is%20Valid
      (Utilities|IsValid _returnvalue
        (:Is%20Valid
          (GameplayEffects|RemoveActiveEffectswithGrantedTags _returnvalue "(GameplayTags=((TagName=\"Ability.Chrono.AncientWeapon.InFlight\")))") ; Shield Visibility
          (CallFunction|KillProjectile))))))

(event Custom|SetMeshVisibility
  (params (Visible :type Boolean)))

(fn ApplyImpulse
  (params (Actor :type Actor%20Object%20Reference))
  (GameplayEffects|ApplyGameplayEffectToTarget (Default|GetInstigatorASC) (Variables|GE&Damage|GetGE_Impulse) (Ability|GetFortAbilitySystemComponent Actor) 0.0 (GameplayEffects|MakeFortEffectContext (Default|GetInstigatorASC) 0 self self 0)))

(fn bIsGroundHitCloseDistance
  (return (and (< (Math|Vector|DistanceSquared(Vector) (Variables|Default|GetStartLocation) (Variables|HitInfo|GetBoomerangHitLoc)) (ScalableFloat|GetValueatLevel (Variables|Bounce|GetRow_BouncePopUpGroundHitDistSqCheck))) (< (.z (Variables|HitInfo|GetBoomerangHitNormal)) (ScalableFloat|GetValueatLevel (Variables|Bounce|GetRow_BounceHitNormalZCheckForPopUp))))))

(fn CalculateKnockbackVelocity
  (params (Actor :type Actor%20Object%20Reference))
  (bind (_ascharacter _bsuccess) (Utilities|Casting|CastToCharacter Actor))
  (return (Math|Vector|MakeVector (.x (* (Math|Vector|Normalize (Math|Vector|GetUnitDirection(Vector) (Transformation|GetActorLocation) (Transformation|GetActorLocation Actor))) (Math|Float|SelectFloat (ScalableFloat|GetValueatLevel (Variables|BPSetVariables|GetRow_PlayerKnockbackStrength)) (ScalableFloat|GetValueatLevel (Variables|BPSetVariables|GetRow_VehicleKnockbackStrength)) _bsuccess))) (.y (* (Math|Vector|Normalize (Math|Vector|GetUnitDirection(Vector) (Transformation|GetActorLocation) (Transformation|GetActorLocation Actor))) (Math|Float|SelectFloat (ScalableFloat|GetValueatLevel (Variables|BPSetVariables|GetRow_PlayerKnockbackStrength)) (ScalableFloat|GetValueatLevel (Variables|BPSetVariables|GetRow_VehicleKnockbackStrength)) _bsuccess))) (+ (.z (* (Math|Vector|Normalize (Math|Vector|GetUnitDirection(Vector) (Transformation|GetActorLocation) (Transformation|GetActorLocation Actor))) (Math|Float|SelectFloat (ScalableFloat|GetValueatLevel (Variables|BPSetVariables|GetRow_PlayerKnockbackStrength)) (ScalableFloat|GetValueatLevel (Variables|BPSetVariables|GetRow_VehicleKnockbackStrength)) _bsuccess))) (Math|Float|SelectFloat (ScalableFloat|GetValueatLevel (Variables|BPSetVariables|GetRow_PlayerKnockbackStrength)) (ScalableFloat|GetValueatLevel (Variables|BPSetVariables|GetRow_VehicleKnockbackStrength)) _bsuccess)))))

(fn CheckDistaneToPopUpandReturnEarly
  (return (and (< (Math|Vector|DistanceSquared(Vector) (Variables|Default|GetStartLocation) (Variables|HitInfo|GetBoomerangHitLoc)) (ScalableFloat|GetValueatLevel (Variables|Bounce|GetRow_BouncePopUpDistSqCheck))) (< (.z (Variables|HitInfo|GetBoomerangHitNormal)) (ScalableFloat|GetValueatLevel (Variables|Bounce|GetRow_BounceHitNormalZCheckForPopUp))))))

(fn CheckImmunityStatus
  (params (Key :type Character%20Object%20Reference%20%28by%20ref%29))
  (bind _recentlyhit (Variables|HitInfo|GetRecentlyHit))
  (bind (_value _returnvalue) (Utilities|Map|Find _recentlyhit Key))
  (if _returnvalue
    (if (> (- (Fortnite|Sync|Time|GetServerWorldTimeSeconds) _value) (ScalableFloat|GetValueatLevel (Variables|BPSetVariables|GetRow_PostHitImmunityDuration)))
      (Development|PrintString "Immunity expired, removing from map")
      (Utilities|Map|Remove (Variables|HitInfo|GetRecentlyHit) Key)
      (return false)
      (else
        (return true)))
    (else
      (return false))))

(fn EvaluateEndingHomingBehavior
  (params (HitPawn :type Character%20Object%20Reference))
  (bind (_homingstyle _minturnspeed _maxturnspeed _ramptimefrommintomaxturnspeed _locktargetdistancethreshold _randomseed _lockedontarget) (Utilities|Struct|BreakProjectileHomingData (Variables|Projectile|GetProjectile%20Homing%20Data)))
  (bind _row_homingeffecttriggerrange (Variables|HomingValues|GetRow_HomingEffectTriggerRange))
  (bind _returnvalue (ScalableFloat|GetValueatLevel _row_homingeffecttriggerrange))
  (bind _row_homingminturnspeed (Variables|HomingValues|GetRow_HomingMinTurnSpeed))
  (bind _returnvalue_2 (ScalableFloat|GetValueatLevel _row_homingminturnspeed))
  (bind _row_homingmaxturnspeed (Variables|HomingValues|GetRow_HomingMaxTurnSpeed))
  (bind _returnvalue_3 (ScalableFloat|GetValueatLevel _row_homingmaxturnspeed))
  (bind _row_homingturnramptime (Variables|HomingValues|GetRow_HomingTurnRampTime))
  (bind _returnvalue_4 (ScalableFloat|GetValueatLevel _row_homingturnramptime))
  (bind _projectilehomingdata (Utilities|Struct|MakeProjectileHomingData "None" _returnvalue_2 _returnvalue_3 _returnvalue_4 _returnvalue))
  (bind _projectilemovementcomponent (Variables|Movement|GetProjectileMovementComponent))
  ;[1 1 1] stop chasing the target if we've hit it
  (if (== HitPawn _lockedontarget)
    (Projectile|SetHomingData _projectilemovementcomponent _projectilehomingdata :!class /Script/FortniteGame.FortProjectileMovementComponent) ; reset the homing parameters to stop following the target
    (Variables|Projectile|SetProjectile%20Homing%20Data _projectilehomingdata)
    (Components|Movement|SetUpdatedComponent _projectilemovementcomponent (Variables|Collision|GetCapsuleComponent))))

(fn FindShortHitLocation
  (bind (_outhit _returnvalue) (Collision|LineTraceByChannel (Transformation|GetWorldLocation (Variables|Collision|GetCapsuleComponent)) (+ (+ (* (Transformation|GetUpVector (Variables|Collision|GetCapsuleComponent)) 1.25e+03) (Transformation|GetWorldLocation (Variables|Collision|GetCapsuleComponent))) (* (Transformation|GetForwardVector (Variables|Collision|GetCapsuleComponent)) -8e+02)) "TraceTypeQuery11"))
  (return (Math|Vector|SelectVector (.Location (Collision|BreakHitResult _outhit)) (.TraceEnd (Collision|BreakHitResult _outhit)) _returnvalue)))

(fn GetDesiredReturnDuration
  ;[1 1 1] TODO: calculate based on distance traveled / remaining between item & pawn. Current flight out time is 1.25s is 15 tiles or 3840 max length
  (return (/ (Variables|Default|GetDistance%20Check%20Based%20on%20Flight%20Distance) (Math|Float|Clamp(Float) (Math|Vector|VectorLength (- (Transformation|GetActorLocation) (Transformation|GetActorLocation (Variables|Owning|GetOwningPlayerPawn)))) 3e+02 (Math|Vector|VectorLength (- (Transformation|GetActorLocation) (Transformation|GetActorLocation (Variables|Owning|GetOwningPlayerPawn))))))))

(fn GetReturnArcActorLocation
  (return (+ (Transformation|GetWorldLocation (Variables|Collision|GetCapsuleComponent)) (Utilities|Operators|Add (* (Transformation|GetRightVector (Variables|Collision|GetCapsuleComponent)) (ScalableFloat|GetValueatLevel (Variables|ReturnArc|GetRow_ReturnArcDistRight))) (* (Transformation|GetForwardVector (Variables|Collision|GetCapsuleComponent)) (ScalableFloat|GetValueatLevel (Variables|ReturnArc|GetRow_ReturnArcDistForward))) (* (Transformation|GetUpVector (Variables|Collision|GetCapsuleComponent)) (ScalableFloat|GetValueatLevel (Variables|ReturnArc|GetRow_ReturnArcDistUp)))))))

(fn GetReturnArcAfterPawnHit
  (return (+ (Transformation|GetWorldLocation (Variables|Collision|GetCapsuleComponent)) (Utilities|Operators|Add (* (Transformation|GetRightVector (Variables|Collision|GetCapsuleComponent)) (ScalableFloat|GetValueatLevel (Variables|ReturnArc|GetRow_AfterHitReturnArcDistRight))) (* (Transformation|GetForwardVector (Variables|Collision|GetCapsuleComponent)) (ScalableFloat|GetValueatLevel (Variables|ReturnArc|GetRow_AfterHitReturnArcDistForward))) (* (Transformation|GetUpVector (Variables|Collision|GetCapsuleComponent)) (ScalableFloat|GetValueatLevel (Variables|ReturnArc|GetRow_AfterHitReturnArcDistUp)))))))

(fn GetInstigatorASC
  (return (Ability|GetFortAbilitySystemComponent (Game|GetInstigator))))

(fn HandleProjectileStopReturn
  (Networking|SwitchHasAuthority
    (:Authority
      (Utilities|Time|SetTimerbyEvent (EventDispatchers|CreateEvent :!event CallReturnMovement) 1e-05))))

(fn HandleEnvironmentalTarget
  (params (Hit :type Hit%20Result%20Structure%20%28by%20ref%29))
  (bind (_bblockinghit _binitialoverlap _time _distance _location _impactpoint _normal _impactnormal _physmat _hitactor) (Collision|BreakHitResult Hit))
  (bind _returnvalue (Ability|GetAbilitySystemComponent _hitactor))
  (Utilities|IsValid _returnvalue
    (:Is%20Valid
      (GameplayEffects|FortApplyGameplayEffectToTarget (Default|GetInstigatorASC) (Variables|GE&Damage|GetChargeEnvDamageGE) _returnvalue 0 self self 0.0 (GameplayEffects|MakeEffectContext (Default|GetInstigatorASC))))))

(fn HandleProjectileStop
  (Networking|SwitchHasAuthority
    (:Authority
      (Utilities|Time|SetTimerbyEvent (EventDispatchers|CreateEvent :!event CallMovement) 1e-05))))

(fn LOSCheck
  (params (TracedActor :type Actor%20Object%20Reference))
  (bind _outhit (Collision|LineTraceByChannel (Transformation|GetActorLocation) (Transformation|GetActorLocation TracedActor) "TraceTypeQuery1" false (Variables|Targeting|GetActorsToIgnoreForTargeting)))
  (if (Utilities|Equal(Object) (.HitActor (Collision|BreakHitResult _outhit)) TracedActor)
    (return true)
    (else
      (return false))))

(fn OnHit
  (params (HitLocation :type Vector) (HitNormal :type Vector))
  (Variables|/Script/Engine.MovementComponent|SetVelocity (Math|Vector|MirrorVectorbyNormal (Variables|/Script/Engine.MovementComponent|GetVelocity (Variables|Movement|GetProjectileMovementComponent)) HitNormal) (Variables|Movement|GetProjectileMovementComponent)))

(fn OnRep_bHit)

(fn OnRep_bReturnTimerExpired
  (if (Variables|Default|GetbReturnTimerExpired)
    (if (Variables|Targeting|GetbTargetAcquired)
      _
      (elif (Variables|ReturnArc|GetHas%20Flight%20Angle%20Changed%20Based%20on%20Impact)
        _
        (elif (> (Variables|Default|GetCurrentNumberOfBounces) 0)
          ;[1 1 1] If we bounce a few times, don't arc back 'naturally'
          (Variables|Default|SetIsReturning true)
          (Variables|Default|SetReturnToSender true) ; Set Return Things
          (CallFunction|PrjReturn false)
          (else
            (CallFunction|TravelReturnArcStart)))))))

(fn OnRep_FirstDistVector)

(fn OnRep_FwdDistVector)

(fn OnRep_HasFlightAngleChangedBasedonImpact)

(fn OnRep_HitLocation)

(fn OnRep_PlayerLocation)

(fn OnRep_ReturnPointVector)

(fn OnRep_ReturnToSender
  (bind _self self)
  (bind _projectilemovementcomponent (Variables|Movement|GetProjectileMovementComponent))
  (Development|PrintString "REP NOTIFY: Return To Sender ")
  (Collision|SetCollisionEnabled (Variables|Collision|GetCapsuleComponent) "NoCollision")
  (Projectile|SetHomingStyle 0 _self "LockOn" false :!class /Script/FortniteGame.FortProjectileMovementInterface)
  (Projectile|SetHomingTarget 0 _self (Variables|Owning|GetOwningPlayerPawn) 0 :!class /Script/FortniteGame.FortProjectileMovementInterface)
  (Projectile|SetHomingLockTargetDistanceThreshold 0 _self (ScalableFloat|GetValueatLevel (Variables|HomingValues|GetRow_ReturnHomingLockTargetDistanceThreshold)) :!class /Script/FortniteGame.FortProjectileMovementInterface)
  (Projectile|SetHomingTargetOffset _self (Variables|HomingValues|GetReturnHomingTargetOffset) :!class /Script/FortniteGame.FortProjectileMovementInterface)
  (Variables|/Script/Engine.ProjectileMovementComponent|SetInitialSpeed (ScalableFloat|GetValueatLevel (Variables|BPSetVariables|GetRow_ReturnStartVelocity)) _projectilemovementcomponent)
  (Variables|/Script/Engine.ProjectileMovementComponent|SetMaxSpeed (ScalableFloat|GetValueatLevel (Variables|BPSetVariables|GetRow_ReturnEndVelocity)) _projectilemovementcomponent)
  (Projectile|SetHomingTurnSpeedMax 0 self (ScalableFloat|GetValueatLevel (Variables|BPSetVariables|GetRow_ReturnMaxTurnSpeed)) :!class /Script/FortniteGame.FortProjectileMovementInterface)
  (Collision|SetCollisionEnabled (Variables|Collision|GetCapsuleComponent) "QueryOnly")
  (Default|ResumeMovement false) ; Do not reset the velocity when returning
  (Variables|Projectile|SetProjectile%20Homing%20Data (Utilities|Struct|MakeProjectileHomingData "None" (ScalableFloat|GetValueatLevel (Variables|HomingValues|GetRow_HomingMinTurnSpeed)) (ScalableFloat|GetValueatLevel (Variables|HomingValues|GetRow_HomingMaxTurnSpeed)) (ScalableFloat|GetValueatLevel (Variables|HomingValues|GetRow_HomingTurnRampTime)) (ScalableFloat|GetValueatLevel (Variables|HomingValues|GetRow_HomingEffectTriggerRange))))
  (CallFunction|ExecuteAudioParameter "OnRecall"))

(fn OnRep_RightDistVector)

(fn OnRep_Target)

(fn PlayHitPawnSounds
  (params (HitPawn :type Fort%20Player%20Pawn%20Object%20Reference))
  (bind _owningplayerpawn (Variables|Owning|GetOwningPlayerPawn))
  (Utilities|IsValid _owningplayerpawn
    (:Is%20Valid
      (if (Pawn|IsLocallyControlled _owningplayerpawn)
        (Audio|PlaySound2D (Variables|Audio|GetHitPlayerSoundCue))
        (elif (Pawn|IsLocallyControlled HitPawn)
          (Audio|PlaySound2D (Variables|Audio|GetHitPlayerSoundCue))
          (else
            (Audio|PlaySoundatLocation (Variables|Audio|GetHitEnvSoundCue) (Transformation|GetActorLocation))))))))

(fn PlayRecallSound
  (bind _owningplayerpawn (Variables|Owning|GetOwningPlayerPawn))
  (Utilities|FlowControl|Sequence
    (:then_0
      ;[1 1 1] Play Recall Whoosh Sound
      (bind _returnvalue (Audio|SpawnSoundAttached (Variables|Audio|GetRecallWhooshSound) (Variables|Collision|GetCapsuleComponent)))
      (Utilities|IsValid _owningplayerpawn
        (:Is%20Valid
          (Audio|Parameter|SetBooleanParameter _returnvalue "IsLocal" (Pawn|IsLocallyViewed _owningplayerpawn) :!class /Script/AudioExtensions.AudioParameterControllerInterface))))
    (:then_1
      ;[1 1 1] Stop cast sound so we can then play recall sound. Valid checks just incase of any kind of spam.
      (bind _castinairaudio (Variables|Default|GetCastInAirAudio)
        (:then
          (Audio|Components|Audio|FadeOut _castinairaudio 0.5 0.0 "Linear" :!class /Script/Engine.AudioComponent))
        (:else
          _))
      (Variables|Default|GetRecallAudio
        (:then)
        (:else
          (bind _returnvalue_2 (Audio|SpawnSoundAttached (Variables|Audio|GetRecallSound) (Variables|Transformation|GetRootComponent (Variables|Owning|GetOwningPlayerPawn))))
          (Variables|Default|SetRecallAudio _returnvalue_2))))))

(fn PrintActorsFromArray
  (params (ActorArray :type Array%20of%20Actor%20Object%20References))
  (for _ ActorArray
    (Development|PrintString (Utilities|String|Append "Hit;" (Utilities|GetDisplayName (Utilities|Array|Get(acopy) ActorArray (LocalVariables|GetHitActorIndex)))) true true "(R=1.000000,G=0.019947,B=0.059748,A=1.000000)" 5.0)
    (LocalVariables|SetHitActorIndex (+ (LocalVariables|GetHitActorIndex) 1))))

(fn PrintReturnInfo
  (Development|PrintString (Utilities|String|ToString(Float) (Math|Vector|Distance(Vector) (Variables|Default|GetStartLocation) (Transformation|GetActorLocation))) true true "(R=0.016631,G=1.000000,B=0.121267,A=1.000000)" 5.0))

(fn Resume_Movement
  (params (SetVelocityToPrevious :type Boolean))
  (Projectile|ResumeSimulation)
  (Collision|SetCollisionEnabled (Variables|Collision|GetCapsuleComponent) "QueryOnly" :!pinned true) ; Make sure the collision is enabled
  (Components|Movement|SetUpdatedComponent (Variables|Movement|GetProjectileMovementComponent) (Variables|Collision|GetCapsuleComponent))
  (if SetVelocityToPrevious
    _))

(fn SetHitEffectContext
  (bind _output_get (Variables|HitInfo|SetHitEffectContext (Class|AthenaFunctionLibrary|MakeSimpleFortEffectContext (Game|GetInstigator) (Variables|Default|GetB%20Space%20Clique%20Weapon))))
  (return _output_get))

(fn SetHomingVars
  (bind _self self)
  (bind _b_space_clique_weapon (Variables|Default|GetB%20Space%20Clique%20Weapon))
  (bind _homingtarget (Variables|/SpaceClique/Gameplay/B_SpaceClique_Weapon.B_SpaceClique_Weapon_C|GetHomingTarget _b_space_clique_weapon))
  (Projectile|SetHomingStyle 0 0 "LockOn" false :!class /Script/FortniteGame.FortProjectileMovementInterface)
  (Projectile|SetHomingTurnSpeedMin 0 _self (ScalableFloat|GetValueatLevel (Variables|HomingValues|GetRow_HomingMinTurnSpeed)) :!class /Script/FortniteGame.FortProjectileMovementInterface)
  (Projectile|SetHomingTurnSpeedMax 0 _self (ScalableFloat|GetValueatLevel (Variables|HomingValues|GetRow_HomingMaxTurnSpeed)) :!class /Script/FortniteGame.FortProjectileMovementInterface)
  (Projectile|SetHomingTimeUntilMaxTurnSpeed 0 _self 0.0 :!class /Script/FortniteGame.FortProjectileMovementInterface)
  (Projectile|SetHomingTargetOffset _self (Variables|Default|GetTargetHomingOffset) :!class /Script/FortniteGame.FortProjectileMovementInterface)
  (Utilities|IsValid _homingtarget
    (:Is%20Valid
      (Projectile|SetHomingTarget 0 self _homingtarget 0 :!class /Script/FortniteGame.FortProjectileMovementInterface))))

(fn SetInAirLoopSound
  (bind _ancientweapon_primary_loop_cue (Variables|Default|GetAncientWeapon_Primary_Loop_Cue))
  (bind _owningplayerpawn (Variables|Owning|GetOwningPlayerPawn))
  (Variables|Default|GetAncientWeapon_Primary_Loop_Cue
    (:then
      (if true
        (Audio|Components|Audio|SetSound _ancientweapon_primary_loop_cue (Variables|Audio|GetInAirLoopMSS))
        (Audio|Parameter|SetIntegerParameter _ancientweapon_primary_loop_cue "LocalCharges" (FortWeapon|GetMagazineAmmoCount (Variables|Default|GetB%20Space%20Clique%20Weapon)) :!class /Script/AudioExtensions.AudioParameterControllerInterface)
        (else
          (Audio|Components|Audio|SetSound (Variables|Default|GetAncientWeapon_Primary_Loop_Cue) (Variables|Audio|GetInAirLoopSoundCue))))
      (Utilities|IsValid _owningplayerpawn
        (:Is%20Valid
          (Audio|Parameter|SetBooleanParameter (Variables|Default|GetAncientWeapon_Primary_Loop_Cue) "IsLocal" (Pawn|IsLocallyViewed _owningplayerpawn) :!class /Script/AudioExtensions.AudioParameterControllerInterface))))
    (:else)))

(fn ConstructionScript
  (|CallParentFunction))
